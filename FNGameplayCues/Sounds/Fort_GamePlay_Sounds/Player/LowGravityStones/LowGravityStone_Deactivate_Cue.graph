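# Sound cue for deactivating a Low Gravity Stone on the player, reusing Gravity Sphere exit sounds.
# A Random node picks between two equally weighted exit waves and feeds a Modulator at the root.
# It plays quietly in the gadget items sound class with status-effect activation attenuation, and a concurrency override limits it to a single instance at a time.
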SoundCue /FNGameplayCues/Sounds/Fort_GamePlay_Sounds/Player/LowGravityStones/LowGravityStone_Deactivate_Cue
root=Modulator_1
nodes [
 {id=Modulator_1, children [Random_0]}
 {id=Random_0, children [WavePlayer_7,WavePlayer_9], properties {weights [1,1]}}
 {id=WavePlayer_7, properties {soundWaveAssetPtr=/FNGameplayCues/Sounds/Fort_Gadgets/GravitySphere/fort_gravity_sphere_exitnew_01}}
 {id=WavePlayer_9, properties {soundWaveAssetPtr=/FNGameplayCues/Sounds/Fort_Gadgets/GravitySphere/fort_gravity_sphere_exitnew_03}}
]
properties {
 attenuationSettings=/Game/Sounds/Attenuation/Fort_Ability_Status_Effect_Activation_Att
 bOverrideConcurrency=true
 maxDistance=6000
 soundClassObject=&Fort_Gadgets_Items
 volumeMultiplier=0.15
 concurrencyOverrides {maxCount=1}
}
