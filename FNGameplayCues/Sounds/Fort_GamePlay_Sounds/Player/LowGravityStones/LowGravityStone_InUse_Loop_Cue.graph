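# Looping sound cue for a low gravity stone in use, playing as a single concurrent instance.
# An IsLocal branch selects stereo or mono attenuation over duplicate chains, each mixing a modulated gravity-sphere loop with an in-air loop.
# Fall-speed parameters drive the in-air layer's volume and pitch and each mix's overall volume.
# A root enveloper fades the sound in, then eases it down to half volume.
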SoundCue /FNGameplayCues/Sounds/Fort_GamePlay_Sounds/Player/LowGravityStones/LowGravityStone_InUse_Loop_Cue
root=Enveloper_0
nodes [
 {
  id=Attenuation_0
  children [ModulatorContinuous_3]
  properties {attenuationSettings=/FNGameplayCues/Sounds/Attenuation/Fort_Ability_Status_Effect_Looping_Stereo_Att}
 }
 {
  id=Attenuation_1
  children [ModulatorContinuous_8]
  properties {attenuationSettings=/FNGameplayCues/Sounds/Attenuation/Fort_Ability_Status_Effect_Looping_Mono_Att}
 }
 {id=Branch_0, children [Attenuation_0, Attenuation_1, Attenuation_1], properties {boolParameterName=IsLocal}}
 {
  id=Enveloper_0
  children [Branch_0]
  properties {pitchCurve {editorCurveData {keys [{time=0}]}}, volumeCurve {editorCurveData {keys [<at 0 0>,<at 0.3 1>,<at 1.75 1>,<at 2.5 0.5>]}}}
 }
 {id=Mixer_1, children [Modulator_2, ModulatorContinuous_0], properties {inputVolume [1,0.8]}}
 {id=Mixer_5, children [Modulator_1, ModulatorContinuous_6], properties {inputVolume [1,0.8]}}
 {
  id=ModulatorContinuous_0
  children [WavePlayer_1]
  properties {
   pitchModulationParams {maxOutput=2, minOutput=1, parameterName=FallSpeedInv}
   volumeModulationParams {maxInput=0.25, parameterName=Fallspeed}
  }
 }
 {
  id=ModulatorContinuous_3
  children [Mixer_1]
  properties {pitchModulationParams {maxOutput=1.25, minOutput=1}, volumeModulationParams {minOutput=0.5, parameterName=FallSpeed}}
 }
 {
  id=ModulatorContinuous_6
  children [WavePlayer_4]
  properties {
   pitchModulationParams {maxOutput=2, minOutput=1, parameterName=FallSpeedInv}
   volumeModulationParams {maxInput=0.25, parameterName=Fallspeed}
  }
 }
 {
  id=ModulatorContinuous_8
  children [Mixer_5]
  properties {pitchModulationParams {maxOutput=1.25, minOutput=1}, volumeModulationParams {minOutput=0.5, parameterName=FallSpeed}}
 }
 {id=Modulator_1, children [WavePlayer_2]}
 {id=Modulator_2, children [WavePlayer_8]}
 {
  id=WavePlayer_1
  properties {bLooping=true, soundWaveAssetPtr=&/LowGravityStone_InAir_Loop}
 }
 {
  id=WavePlayer_2
  properties {bLooping=true, soundWaveAssetPtr=/FNGameplayCues/Sounds/Fort_Gadgets/GravitySphere/fort_gravity_sphere_loop_01Loop}
 }
 {
  id=WavePlayer_4
  properties {bLooping=true, soundWaveAssetPtr=&LowGravityStone_InAir_Loop}
 }
 {
  id=WavePlayer_8
  properties {bLooping=true, soundWaveAssetPtr=&fort_gravity_sphere_loop_01Loop}
 }
]
properties {
 bOverrideConcurrency=true
 maxDistance=3500
 soundClassObject=&Fort_Gadgets_Items
 volumeMultiplier=0.2
 concurrencyOverrides {maxCount=1, resolutionRule=StopFarthestThenPreventNew, retriggerTime=0.3}
}
