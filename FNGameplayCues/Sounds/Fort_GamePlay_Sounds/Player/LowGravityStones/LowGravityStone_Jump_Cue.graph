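# Sound cue for the player's low gravity stone jump, played as a gadget and item effect.
# A random node picks one of three equally weighted jump waves and feeds a modulator that varies pitch and volume slightly on each play.
# It uses the ability status effect activation attenuation and a concurrency set named for jump and land sounds.
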SoundCue /FNGameplayCues/Sounds/Fort_GamePlay_Sounds/Player/LowGravityStones/LowGravityStone_Jump_Cue
root=Modulator_0
nodes [
 {id=Modulator_0, children [Random_0], properties {pitchMax=1.1, pitchMin=0.9, volumeMax=1, volumeMin=0.7}}
 {id=Random_0, children [WavePlayer_0,WavePlayer_1,WavePlayer_2], properties {weights [1,1,1]}}
 {id=WavePlayer_0, properties {soundWaveAssetPtr=/Game/Sounds/Fort_GamePlay_Sounds/Player/LowGravityStones/LowGravityStone_Jump_01}}
 {id=WavePlayer_1, properties {soundWaveAssetPtr=&/LowGravityStone_Jump_02}}
 {id=WavePlayer_2, properties {soundWaveAssetPtr=&/LowGravityStone_Jump_03}}
]
properties {
 attenuationSettings=/Game/Sounds/Attenuation/Fort_Ability_Status_Effect_Activation_Att
 maxDistance=6000
 soundClassObject=&Fort_Gadgets_Items
 volumeMultiplier=1
 concurrencySet [/Game/Sounds/Fort_GamePlay_Sounds/Player/LowGravityStones/LowGravityStone_JumpLand_Concc]
}
