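# Landing sound cue for the Low Gravity Stones player effect, routed to the gadgets/items sound class with attenuation and a shared jump/land concurrency setting.
# A Mixer layers a random Constructor stone landing step at full level with two quieter accents.
# The accents are a random SparkleRemix pickaxe swing with varied pitch and a dedicated low-gravity landing wave pitched up by a fixed amount.
# Each layer passes through its own Modulator.
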SoundCue /FNGameplayCues/Sounds/Fort_GamePlay_Sounds/Player/LowGravityStones/LowGravityStone_Land_Cue
root=Mixer_0
nodes [
 {id=Mixer_0, children [Modulator_0,Modulator_1,Modulator_2], properties {inputVolume [1,0.1,0.3]}}
 {id=Modulator_0, children [Random_3]}
 {id=Modulator_1, children [Random_4], properties {pitchMax=1.1, pitchMin=0.9}}
 {id=Modulator_2, children [WavePlayer_22], properties {pitchMax=1.5, pitchMin=1.5}}
 {id=Random_3, children [WavePlayer_15, WavePlayer_16, WavePlayer_17, WavePlayer_18, WavePlayer_19], properties {weights [1,1,1,1,1]}}
 {id=Random_4, children [WavePlayer_20, WavePlayer_21], properties {weights [1,1]}}
 {id=WavePlayer_15, properties {soundWaveAssetPtr=/FNGameplayCues/Sounds/Fort_Step_Sounds/Constructor/Constructor_Step_Land_Stone_06}}
 {id=WavePlayer_16, properties {soundWaveAssetPtr=/FNGameplayCues/Sounds/Fort_Step_Sounds/Constructor/Constructor_Step_Land_Stone_03}}
 {id=WavePlayer_17, properties {soundWaveAssetPtr=/FNGameplayCues/Sounds/Fort_Step_Sounds/Constructor/Constructor_Step_Land_Stone_05}}
 {id=WavePlayer_18, properties {soundWaveAssetPtr=/FNGameplayCues/Sounds/Fort_Step_Sounds/Constructor/Constructor_Step_Land_Stone_02}}
 {id=WavePlayer_19, properties {soundWaveAssetPtr=/FNGameplayCues/Sounds/Fort_Step_Sounds/Constructor/Constructor_Step_Land_Stone_04}}
 {id=WavePlayer_20, properties {soundWaveAssetPtr=/FNGameplayCues/Athena/Sounds/Weapons/PickAxes/SparkleRemix/PickAxe_SparkleRemix_Swing_01}}
 {id=WavePlayer_21, properties {soundWaveAssetPtr=/FNGameplayCues/Athena/Sounds/Weapons/PickAxes/SparkleRemix/PickAxe_SparkleRemix_Swing_02}}
 {id=WavePlayer_22, properties {soundWaveAssetPtr=&/LowGravityStone_Land_06}}
]
properties {
 attenuationSettings=/Game/Sounds/Attenuation/Fort_Ability_Status_Effect_Activation_Att
 maxDistance=6000
 soundClassObject=&Fort_Gadgets_Items
 volumeMultiplier=0.4
 concurrencySet [/Game/Sounds/Fort_GamePlay_Sounds/Player/LowGravityStones/LowGravityStone_JumpLand_Concc]
}
