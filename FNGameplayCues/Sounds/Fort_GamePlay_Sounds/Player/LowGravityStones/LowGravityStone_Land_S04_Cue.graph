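# LowGravityStone_Land_S04_Cue is a sound cue for a Low Gravity Stone landing sound.
# A Random node picks one of three equally weighted landing waves, then a Modulator applies a fixed upward pitch shift rather than a random range.
# It uses the gadgets and items sound class with reduced volume, status-effect activation attenuation, and a concurrency set named for jump and landing sounds.
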SoundCue /FNGameplayCues/Sounds/Fort_GamePlay_Sounds/Player/LowGravityStones/LowGravityStone_Land_S04_Cue
root=Modulator_2
nodes [
 {id=Modulator_2, children [Random_0], properties {pitchMax=1.5, pitchMin=1.5}}
 {id=Random_0, children [WavePlayer_0, WavePlayer_1, WavePlayer_22], properties {weights [1,1,1]}}
 {id=WavePlayer_0, properties {soundWaveAssetPtr=&/LowGravityStone_Land_05}}
 {id=WavePlayer_1, properties {soundWaveAssetPtr=&/LowGravityStone_Land_04}}
 {id=WavePlayer_22, properties {soundWaveAssetPtr=&/LowGravityStone_Land_06}}
]
properties {
 attenuationSettings=/Game/Sounds/Attenuation/Fort_Ability_Status_Effect_Activation_Att
 maxDistance=6000
 soundClassObject=&Fort_Gadgets_Items
 volumeMultiplier=0.392
 concurrencySet [/Game/Sounds/Fort_GamePlay_Sounds/Player/LowGravityStones/LowGravityStone_JumpLand_Concc]
}
