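# Sound cue for the low gravity stone warning, a player gameplay sound in the gadgets and items sound class.
# A Random node picks among four equally weighted wave players using warning variations 08, 09 and 10.
# Variation 09 fills two of the four slots, so it is weighted double in effect.
# Playback runs at reduced volume with the shared status effect activation attenuation.
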SoundCue /FNGameplayCues/Sounds/Fort_GamePlay_Sounds/Player/LowGravityStones/LowGravityStone_Warning_Cue
root=Random_0
nodes [
 {id=Random_0, children [WavePlayer_3,WavePlayer_4,WavePlayer_1,WavePlayer_5], properties {weights [1,1,1,1]}}
 {
  id=WavePlayer_1
  properties {soundWaveAssetPtr=&/LowGravityStone_Warning_10}
 }
 {
  id=WavePlayer_3
  properties {soundWaveAssetPtr=&/LowGravityStone_Warning_08}
 }
 {
  id=WavePlayer_4
  properties {soundWaveAssetPtr=&/LowGravityStone_Warning_09}
 }
 {
  id=WavePlayer_5
  properties {soundWaveAssetPtr=&LowGravityStone_Warning_09}
 }
]
properties {
 attenuationSettings=/Game/Sounds/Attenuation/Fort_Ability_Status_Effect_Activation_Att
 maxDistance=6000
 soundClassObject=&Fort_Gadgets_Items
 volumeMultiplier=0.5
}
